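# Translucent lit material for dust puff particles on Niagara sprites and meshes.
# A panning normal map distorts the puff texture and scrolling smoke mask UVs.
# Inverted smoke times puff alpha feeds a SmoothStep erosion driven by dynamic Clip and Soft Edge, then gains CenterFill, scales by Opacity and gets depth fade.
# Particle color tints the puff's green channel plus Hot-scaled red channel, feeding an emissive adjustment function.
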Material /DealershipSharedLibrary/Assets/Sets/NexusPlexus/Materials/M_DLR_DustPuff
graph {
 properties {
  bCastRayTracedShadows=false
  bUseTranslucencyVertexFog=false
  bUsedWithNiagaraMeshParticles=true
  bUsedWithNiagaraSprites=true
  blendMode=BLEND_Translucent
  pixelDepthOffsetMode=PDOM_Legacy
  translucencyDirectionalLightingIntensity=0.1
  translucentSelfShadowDensityScale=0
  translucentSelfShadowSecondDensityScale=1
  translucentShadowDensityScale=0
  translucentShadowStartOffset=1
  usedShadingModels=Default Lit
  lightmassSettings {
   bCastShadowAsMasked=false
   bOverrideCastShadowAsMasked=false
   bOverrideDiffuseBoost=false
   bOverrideEmissiveBoost=false
   bOverrideExportResolutionScale=false
   diffuseBoost=8
   emissiveBoost=1
   exportResolutionScale=1
  }
  translucentMultipleScatteringExtinction=<rgba 1 0.8771 0.6778 1>
 }
 nodes [
  Constant3Vector() {constant=<rgba 0.7085 0.6291 0.849 1>}
  n1 = TextureCoordinate()
  n2 = Panner(Coordinate=TextureCoordinate() {uTiling=1.5, vTiling=1.5}) {speedX=0.11, speedY=0.1}
  n3 = Multiply(
   A=Constant() {desc=Refract_Intensity, r=0.02}
   B=Multiply(A=TextureSample(UVs=@n2) {samplerType=SAMPLERTYPE_Normal, texture=/DealershipSharedLibrary/FX/Textures/Normal/T_VCVFX_Liquid_Tile_N}.R)
  )
  n4 = DynamicParameter() {parameterIndex=1, paramNames [PanSpeed,CenterFill,0,0]}
  n5 = TextureSample(UVs=Panner(Coordinate=Add(A=@n1, B=@n3), Time=@n4.PanSpeed) {speedY=0.5}) {desc=Shadow Color, texture=/DealershipSharedLibrary/Assets/Sets/NexusPlexus/Textures/T_DLR_SmokeMask_Plexus}
  n6 = TextureSample(UVs=Add(A=TextureCoordinate(), B=@n3)) {desc=Shadow Color, texture=/DealershipSharedLibrary/Assets/Sets/NexusPlexus/Textures/T_DLR_DustPuff_Plexus}
  n7 = DynamicParameter() {paramNames [Opacity, Clip, Soft Edge, Hot]}
  n8 = OneMinus(None=Multiply(A=Constant() {desc=Clip, r=1}, B=@n7.Clip))
  n9 = Constant() {desc=SoftEdge, r=2}
  n10 = MaterialFunctionCall(Alpha=Multiply(A=OneMinus(None=@n5.RGB), B=@n6.A), Min=@n8, Max=Add(A=@n8, B=Multiply(A=@n9, B=@n7.Soft Edge))) {materialFunction=&SmoothStep}
  n11 = Multiply(A=@n4.CenterFill, B=@n6.A)
  DepthFade(Opacity=Multiply(A=Add(A=@n10.Result, B=@n11), B=@n7.Opacity)) {fadeDistanceDefault=300}
  n12 = ParticleColor()
  n13 = Multiply(A=@n6.R, B=@n12.RGB)
  MaterialFunctionCall(Emissive=Add(A=Multiply(A=@n12.RGB, B=@n6.G), B=Multiply(A=@n13, B=@n7.Hot))) {materialFunction=&MF_FX_EmissiveAdjustments}
 ]
}
