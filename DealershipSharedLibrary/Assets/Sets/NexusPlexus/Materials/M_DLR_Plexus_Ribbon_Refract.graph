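# Unlit additive material for Niagara ribbons and particle sprites in the NexusPlexus set, producing a flowing, distorted noise streak.
# Two scrolling Perlin noise layers, their UVs warped by panning liquid normal maps, are blended and masked by a gradient texture.
# That mask is sharpened, faded along U, scaled by particle alpha and depth-faded for opacity.
# Particle color feeds an emissive adjustment function; a ribbon-lines add and dynamic parameter feed no other node.
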Material /DealershipSharedLibrary/Assets/Sets/NexusPlexus/Materials/M_DLR_Plexus_Ribbon_Refract
graph {
 properties {
  bEnableResponsiveAA=true
  bUsedWithNiagaraRibbons=true
  bUsedWithNiagaraSprites=true
  bUsedWithParticleSprites=true
  blendMode=BLEND_Additive
  pixelDepthOffsetMode=PDOM_Legacy
  refractionMethod=RM_2DOffset
  shadingModel=MSM_Unlit
  usedShadingModels=Unlit
 }
 nodes [
  n1 = Panner(Coordinate=TextureCoordinate() {uTiling=0.5, vTiling=0.5}) {speedX=0.1}
  n2 = Multiply(
   A=Constant() {desc=Refract_Intensity, r=0.1}
   B=Multiply(A=TextureSample(UVs=@n1) {samplerType=SAMPLERTYPE_Normal, texture=/DealershipSharedLibrary/FX/Textures/Normal/T_VCVFX_Liquid_Tile_N}.G)
  )
  n3 = Panner(Coordinate=Add(A=TextureCoordinate() {uTiling=2}, B=@n2)) {speedX=-1, speedY=0.1}
  n4 = Constant() {desc=Refract_Intensity, r=0.1}
  n5 = TextureCoordinate() {uTiling=0.5, vTiling=0.5}
  n6 = Multiply(A=TextureSample(UVs=Panner(Coordinate=@n5) {speedX=0.1, speedY=0.1}) {samplerType=SAMPLERTYPE_Normal, texture=&T_VCVFX_Liquid_Tile_N}.G)
  n7 = Panner(Coordinate=Add(A=TextureCoordinate() {uTiling=3}, B=Multiply(A=@n4, B=@n6))) {speedX=-1.5, speedY=-0.1}
  n8 = Multiply(
   A=LinearInterpolate(
    A=TextureSample(UVs=@n3) {texture=/DealershipSharedLibrary/FX/Textures/Noise/T_VCVFX_Boost_S16_PerlinBlur}.RGB
    B=TextureSample(UVs=@n7) {texture=&T_VCVFX_Boost_S16_PerlinBlur}.RGB
   )
   B=TextureSample() {texture=/DealershipSharedLibrary/FX/Textures/Gradient_Masks/T_Gradient_LuminousWrap_A}.RGB
  )
  Add(
   A=@n8
   B=TextureSample(UVs=TextureCoordinate()) {texture=/DealershipSharedLibrary/Assets/Sets/NexusPlexus/Textures/T_DLR_Plexus_RibbonLines}.RGB
  )
  Add(A=TextureCoordinate(), B=@n2)
  n9 = ParticleColor()
  n10 = Power(Base=@n8) {constExponent=1.5}
  n11 = OneMinus(None=ComponentMask(None=TextureCoordinate()) {r=true})
  DepthFade(Opacity=Multiply(A=@n9.A, B=Multiply(A=@n10, B=@n11))) {fadeDistanceDefault=150}
  DynamicParameter() {parameterIndex=1, paramNames [PanSpeed,CenterFill,0,0]}
  MaterialFunctionCall(Emissive=@n9.RGB) {materialFunction=&MF_FX_EmissiveAdjustments}
 ]
}
